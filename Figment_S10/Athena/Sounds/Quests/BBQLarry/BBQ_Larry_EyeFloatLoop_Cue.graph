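# Sound cue for a looping eye-float sound tied to the BBQ Larry quest content.
# A single looping wave player feeds the root directly, with no randomization or mixing.
# It uses a short-range quest-interactable hint attenuation and the in-game mission sound class, with a boosted volume.
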SoundCue /Figment_S10/Athena/Sounds/Quests/BBQLarry/BBQ_Larry_EyeFloatLoop_Cue
root=WavePlayer_0
nodes [{id=WavePlayer_0, properties {bLooping=true, soundWaveAssetPtr=&/BBQ_Larry_EyeFloatLoop}}]
properties {
 attenuationSettings=/Figment_S10/Sounds/Attenuation/Fort_QuestInteractables_HintShortRange
 maxDistance=1500
 soundClassObject=/Game/Sounds/FortSoundClassesAndModes/Fort_Mission_IngameSounds
 volumeMultiplier=1.5
}
