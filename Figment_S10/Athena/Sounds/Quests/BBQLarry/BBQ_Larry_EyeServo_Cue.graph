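# Sound cue for the eye servo sound of BBQ Larry, a quest element in Figment Season 10.
# A Random node picks one of three equally weighted servo recordings and feeds it into a root Modulator at default settings for slight variation.
# It plays through the mission in-game sound class with a short-range quest interactable hint attenuation.
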SoundCue /Figment_S10/Athena/Sounds/Quests/BBQLarry/BBQ_Larry_EyeServo_Cue
root=Modulator_0
nodes [
 {id=Modulator_0, children [Random_0]}
 {id=Random_0, children [WavePlayer_1,WavePlayer_2,WavePlayer_3], properties {weights [1,1,1]}}
 {id=WavePlayer_1, properties {soundWaveAssetPtr=&/BBQ_Larry_EyeServo_01}}
 {id=WavePlayer_2, properties {soundWaveAssetPtr=&/BBQ_Larry_EyeServo_02}}
 {id=WavePlayer_3, properties {soundWaveAssetPtr=&/BBQ_Larry_EyeServo_03}}
]
properties {
 attenuationSettings=/Figment_S10/Sounds/Attenuation/Fort_QuestInteractables_HintShortRange
 maxDistance=1500
 soundClassObject=/Game/Sounds/FortSoundClassesAndModes/Fort_Mission_IngameSounds
}
